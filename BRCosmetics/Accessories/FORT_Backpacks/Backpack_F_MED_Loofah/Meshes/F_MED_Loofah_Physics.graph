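# Physics asset for the F_MED_Loofah backpack cosmetic, defining bodies and joints for its dangling props.
# A box body on the pack bone anchors bunny, banana and tomato chains, plus a bottle, two zippers and the bag flap.
# Chains branch into bunny ears, banana skins and tomato leaves, mostly with tight swing and twist limits.
# The zippers swing widely, while the flap locks twist and one swing axis to act as a hinge.
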PhysicsAsset /BRCosmetics/Accessories/FORT_Backpacks/Backpack_F_MED_Loofah/Meshes/F_MED_Loofah_Physics
bodies {
 dyn_bag_flap [{shapeType=Box, extentX=7.5, extentY=10, extentZ=2, center=<1 0 0>}]
 dyn_banana_01 [{shapeType=Capsule, length=3, radius=2.5, center=<2 0 0>, rotation=<rot 90 0 0>}]
 dyn_banana_02 [{shapeType=Capsule, length=3, radius=2.5, center=<2 0 0>, rotation=<rot 90 0 0>}]
 dyn_banana_skin_01 [{shapeType=Capsule, length=1, radius=1, center=<1 0 0>, rotation=<rot 90 0 0>}]
 dyn_banana_skin_02 [{shapeType=Capsule, length=1, radius=1, center=<1 0 0>, rotation=<rot 90 0 0>}]
 dyn_bottle [{shapeType=Capsule, length=12, radius=4, center=<3 0 0>, rotation=<rot 90 0 0>}]
 dyn_bunny_01 [{shapeType=Capsule, length=2.5, radius=4.5, center=<4.6 0 0>, rotation=<rot 90 90 0>}]
 dyn_bunny_ear_l_01 [{shapeType=Box, extentX=5, extentY=4, extentZ=2, center=<2 0 0>}]
 dyn_bunny_ear_l_02 [{shapeType=Box, extentX=5, extentY=4, extentZ=2, center=<2 0 0>}]
 dyn_bunny_ear_r_01 [{shapeType=Box, extentX=5, extentY=4, extentZ=2, center=<2 0 0>}]
 dyn_bunny_ear_r_02 [{shapeType=Box, extentX=5, extentY=4, extentZ=2, center=<2 0 0>}]
 dyn_tomato_01 [{shapeType=Capsule, length=1.5, radius=3.5, center=<3 0 0>, rotation=<rot 90 0 0>}]
 dyn_tomato_02 [{shapeType=Capsule, length=2, radius=5, center=<5 0 0>, rotation=<rot 90 0 -90>}]
 dyn_tomato_leaves_01 [{shapeType=Capsule, length=1, radius=1, center=<1 0 0>, rotation=<rot 90 0 0>}]
 dyn_tomato_leaves_02 [{shapeType=Capsule, length=1, radius=1, center=<1 0 0>, rotation=<rot 90 0 0>}]
 dyn_tomato_leaves_03 [{shapeType=Capsule, length=1, radius=1, center=<1 0 0>, rotation=<rot 90 0 0>}]
 dyn_zipper_l [{shapeType=Capsule, length=2, radius=1, center=<4 0 0>, rotation=<rot 90 0 0>}]
 dyn_zipper_r [{shapeType=Capsule, length=2, radius=1, center=<4 0 0>, rotation=<rot 90 0 0>}]
 pack [{shapeType=Box, extentX=21.2029, extentY=15.2945, extentZ=20.2187, center=<0 -9.4681 -10.4914>}]
}
constraints [
 {bone1Name=dyn_bunny_01, bone2Name=pack, swing1LimitDegrees=25, swing2LimitDegrees=15, twistLimitDegrees=5}
 {bone1Name=dyn_bunny_ear_r_01, bone2Name=dyn_bunny_01, swing1LimitDegrees=30, swing2LimitDegrees=15, twistLimitDegrees=5}
 {bone1Name=dyn_bunny_ear_r_02, bone2Name=dyn_bunny_ear_r_01, swing1LimitDegrees=30, swing2LimitDegrees=15, twistLimitDegrees=5}
 {bone1Name=dyn_bunny_ear_l_01, bone2Name=dyn_bunny_01, swing1LimitDegrees=30, swing2LimitDegrees=15, twistLimitDegrees=5}
 {bone1Name=dyn_bunny_ear_l_02, bone2Name=dyn_bunny_ear_l_01, swing1LimitDegrees=30, swing2LimitDegrees=15, twistLimitDegrees=5}
 {bone1Name=dyn_banana_01, bone2Name=pack, swing1LimitDegrees=5, swing2LimitDegrees=5, twistLimitDegrees=5}
 {bone1Name=dyn_banana_02, bone2Name=dyn_banana_01, swing1LimitDegrees=10, swing2LimitDegrees=10, twistLimitDegrees=5}
 {bone1Name=dyn_banana_skin_01, bone2Name=dyn_banana_02, swing1LimitDegrees=30, swing2LimitDegrees=30, twistLimitDegrees=5}
 {bone1Name=dyn_banana_skin_02, bone2Name=dyn_banana_02, swing1LimitDegrees=30, swing2LimitDegrees=30, twistLimitDegrees=5}
 {bone1Name=dyn_tomato_01, bone2Name=pack, swing1LimitDegrees=10, swing2LimitDegrees=10, twistLimitDegrees=5}
 {bone1Name=dyn_tomato_02, bone2Name=dyn_tomato_01, swing1LimitDegrees=10, swing2LimitDegrees=10, twistLimitDegrees=5}
 {bone1Name=dyn_tomato_leaves_01, bone2Name=dyn_tomato_02, swing1LimitDegrees=10, swing2LimitDegrees=20, twistLimitDegrees=5}
 {bone1Name=dyn_tomato_leaves_02, bone2Name=dyn_tomato_02, swing1LimitDegrees=10, swing2LimitDegrees=20, twistLimitDegrees=5}
 {bone1Name=dyn_tomato_leaves_03, bone2Name=dyn_tomato_02, swing1LimitDegrees=10, swing2LimitDegrees=20, twistLimitDegrees=5}
 {bone1Name=dyn_zipper_l, bone2Name=pack, swing1LimitDegrees=90, swing2LimitDegrees=90, twistLimitDegrees=10}
 {bone1Name=dyn_zipper_r, bone2Name=pack, swing1LimitDegrees=90, swing2LimitDegrees=90, twistLimitDegrees=10}
 {
  bone1Name=dyn_bag_flap
  bone2Name=pack
  swing1LimitDegrees=45
  swing1Motion=Locked
  swing2LimitDegrees=60
  twistLimitDegrees=45
  twistMotion=Locked
 }
 {bone1Name=dyn_bottle, bone2Name=pack, swing1LimitDegrees=15, swing2LimitDegrees=5, twistLimitDegrees=2}
]
